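# Physics asset for the CAU face-accessory part of the medium female player body York, covering its hair bun and side strands.
# Capsule bodies cover neck, head, hair root, bun, a three-segment left strand and a two-segment right strand, each strand with a base body.
# Tight angular limits join the bun to the hair root and link each strand's segments, while neck-to-head, head-to-hair and hair-to-strand-base joints swing and twist freely.
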PhysicsAsset /Game/Characters/Player/Female/Medium/Bodies/F_MED_York/Meshes/Parts/CAU_parts/F_MED_York_CAU_FaceAcc_Physics
bodies {
 cau_bun [{shapeType=Capsule, length=0.5, radius=2.25, center=<-1.1305 0 0>, rotation=<rot 90 0 0>}]
 cau_hair [{shapeType=Capsule, length=1, radius=1}]
 cau_hair_a_base_l [{shapeType=Capsule, length=0.505, radius=0.505, rotation=<rot 0 -90 0>}]
 cau_hair_a_base_r [{shapeType=Capsule, length=0.505, radius=0.505, rotation=<rot 0 -90 0>}]
 cau_hair_a_l [{shapeType=Capsule, length=2.75, radius=1, center=<2.3259 0 0>, rotation=<rot 90 0 0>}]
 cau_hair_a_r [{shapeType=Capsule, length=2.75, radius=1, center=<2.2121 -0.0001 0>, rotation=<rot 90 0 0>}]
 cau_hair_b_l [{shapeType=Capsule, length=1.5, radius=1, center=<1.7404 0 0>, rotation=<rot 90 0 0>}]
 cau_hair_b_r [{shapeType=Capsule, length=3.75, radius=0.75, center=<2.6198 0 0>, rotation=<rot 90 0 0>}]
 cau_hair_c_l [{shapeType=Capsule, length=2, radius=1.25, center=<1.9045 0 0>, rotation=<rot 90 0 0>}]
 head [{shapeType=Capsule, length=6.75, radius=5.5, center=<2.8777 3.22 0>, rotation=<rot 90 0 0>}]
 neck_02 [{shapeType=Capsule, length=4.75, radius=4, center=<0 0.2825 0>, rotation=<rot 90 0 0>}]
}
constraints [
 {bone1Name=cau_bun, bone2Name=cau_hair, swing1LimitDegrees=5, swing2LimitDegrees=5, twistLimitDegrees=5}
 {
  bone1Name=cau_hair
  bone2Name=head
  swing1LimitDegrees=45
  swing1Motion=Free
  swing2LimitDegrees=45
  swing2Motion=Free
  twistLimitDegrees=45
  twistMotion=Free
 }
 {
  bone1Name=head
  bone2Name=neck_02
  swing1LimitDegrees=45
  swing1Motion=Free
  swing2LimitDegrees=45
  swing2Motion=Free
  twistLimitDegrees=45
  twistMotion=Free
 }
 {bone1Name=cau_hair_a_l, bone2Name=cau_hair_a_base_l, swing1LimitDegrees=1, swing2LimitDegrees=1, twistLimitDegrees=1}
 {bone1Name=cau_hair_b_l, bone2Name=cau_hair_a_l, swing1LimitDegrees=5, swing2LimitDegrees=5, twistLimitDegrees=5}
 {bone1Name=cau_hair_c_l, bone2Name=cau_hair_b_l, swing1LimitDegrees=5, swing2LimitDegrees=5, twistLimitDegrees=5}
 {bone1Name=cau_hair_a_r, bone2Name=cau_hair_a_base_r, swing1LimitDegrees=5, swing2LimitDegrees=5, twistLimitDegrees=5}
 {bone1Name=cau_hair_b_r, bone2Name=cau_hair_a_r, swing1LimitDegrees=5, swing2LimitDegrees=5, twistLimitDegrees=5}
 {
  bone1Name=cau_hair_a_base_l
  bone2Name=cau_hair
  swing1LimitDegrees=45
  swing1Motion=Free
  swing2LimitDegrees=45
  swing2Motion=Free
  twistLimitDegrees=45
  twistMotion=Free
 }
 {
  bone1Name=cau_hair_a_base_r
  bone2Name=cau_hair
  swing1LimitDegrees=45
  swing1Motion=Free
  swing2LimitDegrees=45
  swing2Motion=Free
  twistLimitDegrees=45
  twistMotion=Free
 }
]
