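# Post-process style Animation Blueprint for the face accessory (hair and bun) part of the medium female York character, copying the attached parent mesh's pose.
# Additive bone offsets on the left and right hair bases scale with head yaw relative to spine_05, stored as headUp.
# A partial-strength RigidBody simulation follows, then fixed hair and bun nudges.
# Extra bun offsets apply when skydiving or when JumpCheck flags non-zero vertical pawn velocity.
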AnimBlueprint /Game/Characters/Player/Female/Medium/Bodies/F_MED_York/Meshes/Parts/CAU_parts/F_MED_York_CAU_FaceAcc_AnimBP
target_skeleton=&/F_MED_York_CAU_FaceAcc_Skeleton
graphs [
 AnimGraph {nodes [
   n1 = Get headUp() -> (headUp:double)
   n2 = "Transform (Modify) Bone - Bone: cau_hair_a_base_l"(
    ComponentPose=Local To Component(LocalPose=Copy Pose From Mesh() {bUseAttachedParent=true})
    Alpha=@n1
    Translation="0,0,-1"
    Rotation="-30,0,0"
    Scale="1,1,1"
   ) {
    rotationMode=BMM_Additive
    rotationSpace=BCS_BoneSpace
    translationMode=BMM_Additive
    translationSpace=BCS_BoneSpace
    alphaBoolBlend {blendOption=Linear}
    alphaScaleBiasClamp {bMapRange=true, inRange {max=15}}
    boneToModify {boneName=cau_hair_a_base_l}
    rotation {pitch=-30}
    translation {z=-1}
   }
   n3 = "Transform (Modify) Bone - Bone: cau_hair_a_base_r"(ComponentPose=@n2, Alpha=@n1, Translation="0,1,0", Rotation="0,30,0", Scale="1,1,1") {
    rotationMode=BMM_Additive
    rotationSpace=BCS_BoneSpace
    translationMode=BMM_Additive
    translationSpace=BCS_BoneSpace
    alphaBoolBlend {blendOption=Linear}
    alphaScaleBiasClamp {bMapRange=true, inRange {max=15}}
    boneToModify {boneName=cau_hair_a_base_r}
    rotation {yaw=30}
    translation {y=1}
   }
   n4 = RigidBody(ComponentPose=@n3, Alpha=0.4) {
    bClampLinearTranslationLimitToRefPose=true
    bOverrideWorldGravity=true
    overridePhysicsAsset=&/F_MED_York_CAU_FaceAcc_Physics
    simulationSpace=BaseBoneSpace
    alphaBoolBlend {blendOption=Linear}
    baseBoneRef {boneName=root}
    componentLinearAccScale=<0.375 0.375 0.375>
    componentLinearVelScale=<0.375 0.375 0.375>
    overrideWorldGravity {z=-980}
   }
   n5 = "Transform (Modify) Bone - Bone: cau_hair_a_l"(ComponentPose=@n4, Alpha=1, Translation="0,1,0", Rotation="0,0,0", Scale="1,1,1") {
    rotationMode=BMM_Additive
    rotationSpace=BCS_BoneSpace
    translationMode=BMM_Additive
    translationSpace=BCS_BoneSpace
    alphaBoolBlend {blendOption=Linear}
    boneToModify {boneName=cau_hair_a_l}
    translation {y=1}
   }
   n6 = "Transform (Modify) Bone - Bone: cau_bun"!(ComponentPose=@n5, bAlphaBoolEnabled:bool=True, Translation="-1,-4,0", Rotation="0,0,0", Scale="1,1,1") -> (Pose) {
    alphaInputType=Bool
    rotationSpace=BCS_BoneSpace
    translationMode=BMM_Additive
    translationSpace=BCS_BoneSpace
    alphaBoolBlend {blendOption=Linear}
    boneToModify {boneName=cau_bun}
    translation=<-1 -4>
   }
   n7 = "Transform (Modify) Bone - Bone: cau_bun"(ComponentPose=@n6, bAlphaBoolEnabled=Get bIsSkyDiving(), Translation="-2,-5,0", Rotation="0,0,0", Scale="1,1,1") {
    alphaInputType=Bool
    rotationSpace=BCS_BoneSpace
    translationMode=BMM_Additive
    translationSpace=BCS_BoneSpace
    alphaBoolBlend {blendOption=Linear}
    boneToModify {boneName=cau_bun}
    translation=<-2 -5>
   }
   n8 = "Transform (Modify) Bone - Bone: cau_bun"(ComponentPose=@n7, bAlphaBoolEnabled=Get JumpCheck() -> (JumpCheck:bool), Translation="-2,-5,0", Rotation="0,0,0", Scale="1,1,1") {
    alphaInputType=Bool
    rotationSpace=BCS_BoneSpace
    translationMode=BMM_Additive
    translationSpace=BCS_BoneSpace
    alphaBoolBlend {blendOption=Linear}
    boneToModify {boneName=cau_bun}
    translation=<-2 -5>
   }
   Output Pose(Result=Component To Local(ComponentPose=@n8)) {layerGroup=None}
  ]}
 EventGraph {nodes [
   Event BlueprintUpdateAnimation()
   > Sequence()
   n1 = GetDeltaTransformFromRefPose(self=GetOwnerSkeletalMeshComponent(), BoneName=head, BaseName=spine_05)
   > Set headUp(headUp:double=BreakRotator(InRot=BreakTransform(InTransform=@n1).Rotation).Yaw) -> (Output_Get:double)
   n2 = TryGetPawnOwner()
   n3 = Branch(Condition="Not Equal (Float)"(A=GetVelocity!(self:Actor Object Reference=@n2) -> (ReturnValue_X:double, ReturnValue_Y:double, ReturnValue_Z:double).ReturnValue_Z))
   > Set JumpCheck(JumpCheck:bool=true) -> (Output_Get:bool)
   Set JumpCheck(execute=@n3.else, JumpCheck=false)
  ]}
]
variables [
 headUp {type=number}
 JumpCheck {type=boolean}
]
